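# Sound cue for a third-person firework explosion, attenuated for in-world playback.
# A single wave player plays the medium firework explosion sample with boosted volume, using small-firework attenuation and the third-person ranged weapon fire sound class.
# Concurrency rules limit it to one instance per owner, preventing new plays while one is active.
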SoundCue /Figment_S05/Sounds/Fort_GamePlay_Sounds/Explosives/Firework_Explosion_01_Attenuated_Cue
root=WavePlayer_0
nodes [{id=WavePlayer_0, properties {soundWaveAssetPtr=/Game/Sounds/Fort_GamePlay_Sounds/Explosives/Firework_Medium_Explosion_01}}]
properties {
 attenuationSettings=/Game/Sounds/Attenuation/Fort_Weapon_Fireworks_Small_Att
 maxDistance=2500
 soundClassObject=&Fort_WeaponRanged_Fire_3P_PSM
 volumeMultiplier=1.4
 concurrencyOverrides {bLimitToOwner=true, maxCount=1, resolutionRule=PreventNew}
 concurrencySet [/Figment_S05/Sounds/Concurrency_Rules/OneMaximum_Owner_PreventNew]
}
